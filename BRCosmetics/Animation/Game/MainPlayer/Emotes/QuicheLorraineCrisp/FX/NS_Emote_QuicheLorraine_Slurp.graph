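# Niagara system for the QuicheLorraine slurp emote, showing liquid arcing between sockets, sloshing as a mesh, and throwing timed drops and splashes.
# LiquidTrail ribbons and a TrailDropsLead particle follow socket-to-socket Hermite splines, with Minimal spawning droplets from the lead.
# The Liquid mesh uses a viscosity-weighted spring pendulum, feeding orientation, speed and turbulence to dynamic material parameters.
# GlassDrops and Splashes spawn at a socket in delayed windows, and time-of-day colors tint the effect.
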NiagaraSystem /BRCosmetics/Animation/Game/MainPlayer/Emotes/QuicheLorraineCrisp/FX/NS_Emote_QuicheLorraine_Slurp
dependencies {
 usedDataInterfaces [
  &NiagaraDataInterfaceCurve
  &NiagaraDataInterfaceSpriteRendererInfo
  &NiagaraDataInterfaceParticleRead
  &NiagaraDataInterfaceSocketReader
  &NiagaraDataInterfaceSkeletalMesh
 ]
 usedDynamicInputs []
 usedModules [
  &self:SystemSpawnScript.SystemScriptSource.SystemScriptGraph.NiagaraNodeAssignment_0.SetVariables_#1
  &self:SystemSpawnScript.SystemScriptSource.SystemScriptGraph.NiagaraNodeAssignment_1.SetVariables_#2
  &self:SystemSpawnScript.SystemScriptSource.SystemScriptGraph.NiagaraNodeAssignment_2.SetVariables_#3
  &self:SystemSpawnScript.SystemScriptSource.SystemScriptGraph.NiagaraNodeAssignment_3.SetVariables_#4
  &SystemState
 ]
}
emitters {
 GlassDrops {
  emitterUpdateScript [
   &EmitterState(Life Cycle Mode=::NewEnumerator1, Loop Behavior=::NewEnumerator1, Loop Delay=0.55, Loop Duration=1.25, UseLoopDelay=true)
   &SpawnBurst_Instantaneous(Spawn Count=15, Use Spawn Probability=false, UseLoopCountLimit=false)
   &SpawnRate(SpawnRate=60, Use Spawn Probability=false)
  ]
  particleSpawnScript [
   &InitializeParticle(
    Color Mode=::NewEnumerator0
    Lifetime Max=0.3
    Lifetime Min=0.15
    Lifetime Mode=::NewEnumerator1
    Sprite Size Mode=::NewEnumerator1
    Uniform Sprite Size Max=1.6
    Uniform Sprite Size Min=0.9
    UsePositionOffset=false
    Write Material Random=true
   )
   &SocketLocation(SamplePosOffset=false, Socket Index=1, Socket Reader=$System.Socket Reader, Socket Sampling Mode=::NewEnumerator3, VelocitySpeedLimit=false)
   &ShapeLocation(MaskSpawnGroup=false, Override Seed=false, Sphere Radius=3, UseOwnerScale=true)
   &AddVelocity(
    Module.Use Falloff Distance=false
    Origin Offset=<0 0 -8>
    Override Seed=false
    Velocity Mode=::NewEnumerator1
    Velocity Origin=$Output.SocketLocation.SampledPosition
    Velocity Speed=&RandomRangeFloat(Minimum=75, Maximum=130)
   )
  ]
  particleUpdateScript [
   &ParticleState()
   &ScaleSpriteSize(Uniform Curve Sprite Scale=DataInterface() {propertyValues {Curve {
       defaultValue=0
       keys [
        {arriveTangent=9.0725, interpMode=RCIM_Cubic, leaveTangent=9.0725, tangentMode=RCTM_User, time=0, value=0}
        {arriveTangent=0.2183, interpMode=RCIM_Cubic, leaveTangent=0.2183, tangentMode=RCTM_User, time=0.1408, value=0.9847}
        {arriveTangent=-3.9306, interpMode=RCIM_Cubic, leaveTangent=-3.9306, tangentMode=RCTM_User, time=1, value=0}
       ]
      }}})
   &GravityForce(Gravity=<0 0 -850>)
   &SolveForcesAndVelocity(Clamp Velocity=false, Limit Acceleration=false)
   &ScaleSpriteSizeBySpeed(Max Scale Factor=<1.6 5>, Min Scale Factor=<0.8 3>)
   &ScaleColor(
    Scale RGB=&MakeVectorFromLinearColorRGB(Color=&FNTimeOfDayColor(Night Color=<rgba 0.162 0.3316 0.5017 1>, Morning Color=<rgba 1 0.7745 0.7008 1>, Day Color=<rgba 1 1 1 1>, Evening Color=<rgba 1 1 1 1>))
    ScaleA=true
    ScaleRGB=true
   )
  ]
  rendererClasses [&NiagaraSpriteRendererProperties]
  renderers [{rendererClass=&NiagaraSpriteRendererProperties}]
 }
 GlassDrops001 {
  emitterUpdateScript [
   &EmitterState(Life Cycle Mode=::NewEnumerator1, Loop Behavior=::NewEnumerator1, Loop Delay=1.2, Loop Duration=1, UseLoopDelay=true)
   &SpawnPerUnit(
    Max Movement Threshold=100000
    Module.Use Max Movement Threshold=true
    Module.Use Movement Tolerance=false
    Spawn Spacing=3
    Use Spawn Probability=false
   )
  ]
  particleSpawnScript [
   &InitializeParticle(
    Color Mode=::NewEnumerator0
    Lifetime Max=0.3
    Lifetime Min=0.15
    Lifetime Mode=::NewEnumerator1
    Sprite Size Mode=::NewEnumerator1
    Uniform Sprite Size Max=1.6
    Uniform Sprite Size Min=0.9
    UsePositionOffset=false
    Write Material Random=true
   )
   &SocketLocation(
    SamplePosOffset=true
    Sampled Position Offset=<0 0 4>
    Socket Index=1
    Socket Reader=$System.Socket Reader
    Socket Sampling Mode=::NewEnumerator3
    VelocitySpeedLimit=false
   )
   &ShapeLocation(MaskSpawnGroup=false, Override Seed=false, Sphere Radius=3, UseOwnerScale=true)
   &AddVelocity(
    Module.Use Falloff Distance=false
    Origin Offset=<0 0 -6>
    Override Seed=false
    Velocity Mode=::NewEnumerator1
    Velocity Origin=$Output.SocketLocation.SampledPosition
    Velocity Speed=&RandomRangeFloat(Minimum=25, Maximum=60)
   ) {enabled=false}
  ]
  particleUpdateScript [
   &ParticleState()
   &ScaleSpriteSize(Uniform Curve Sprite Scale=DataInterface() {propertyValues {Curve {
       defaultValue=0
       keys [
        {arriveTangent=9.0725, interpMode=RCIM_Cubic, leaveTangent=9.0725, tangentMode=RCTM_User, time=0, value=0}
        {arriveTangent=0.2183, interpMode=RCIM_Cubic, leaveTangent=0.2183, tangentMode=RCTM_User, time=0.1408, value=0.9847}
        {arriveTangent=-3.9306, interpMode=RCIM_Cubic, leaveTangent=-3.9306, tangentMode=RCTM_User, time=1, value=0}
       ]
      }}})
   &GravityForce(Gravity=<0 0 -400>)
   &SolveForcesAndVelocity(Clamp Velocity=false, Limit Acceleration=false)
   &ScaleSpriteSizeBySpeed(Max Scale Factor=<1.6 5>, Min Scale Factor=<0.8 3>) {enabled=false}
   &ScaleColor(
    Scale RGB=&MakeVectorFromLinearColorRGB(Color=&FNTimeOfDayColor(Night Color=<rgba 0.162 0.3316 0.5017 1>, Morning Color=<rgba 1 0.7745 0.7008 1>, Day Color=<rgba 1 1 1 1>, Evening Color=<rgba 1 1 1 1>))
    ScaleA=true
    ScaleRGB=true
   )
  ]
  rendererClasses [&NiagaraSpriteRendererProperties]
  renderers [{rendererClass=&NiagaraSpriteRendererProperties}]
 }
 GlassDrops002 {
  emitterUpdateScript [
   &EmitterState(Life Cycle Mode=::NewEnumerator1, Loop Behavior=::NewEnumerator1, Loop Delay=2.6, Loop Duration=0.5, UseLoopDelay=true)
   &SpawnRate(SpawnRate=46, Use Spawn Probability=false)
  ]
  particleSpawnScript [
   &InitializeParticle(
    Color Mode=::NewEnumerator0
    Lifetime Max=0.25
    Lifetime Min=0.15
    Lifetime Mode=::NewEnumerator1
    Sprite Size Mode=::NewEnumerator1
    Uniform Sprite Size Max=1
    Uniform Sprite Size Min=0.45
    UsePositionOffset=false
    Write Material Random=true
   )
   &SocketLocation(SamplePosOffset=false, Socket Index=1, Socket Reader=$System.Socket Reader, Socket Sampling Mode=::NewEnumerator3, VelocitySpeedLimit=false)
   &ShapeLocation(MaskSpawnGroup=false, Override Seed=false, Sphere Radius=1, UseOwnerScale=true)
   &AddVelocity(
    Module.Use Falloff Distance=false
    Origin Offset=<0 0 -6>
    Override Seed=false
    Velocity Mode=::NewEnumerator1
    Velocity Origin=$Output.SocketLocation.SampledPosition
    Velocity Speed=&RandomRangeFloat(Minimum=25, Maximum=60)
   ) {enabled=false}
   &CameraOffset(Camera Offset Amount=5)
  ]
  particleUpdateScript [
   &ParticleState()
   &ScaleSpriteSize(Uniform Curve Sprite Scale=DataInterface() {propertyValues {Curve {
       defaultValue=0
       keys [
        {arriveTangent=9.0725, interpMode=RCIM_Cubic, leaveTangent=9.0725, tangentMode=RCTM_User, time=0, value=0}
        {arriveTangent=0.2183, interpMode=RCIM_Cubic, leaveTangent=0.2183, tangentMode=RCTM_User, time=0.1408, value=0.9847}
        {arriveTangent=-3.9306, interpMode=RCIM_Cubic, leaveTangent=-3.9306, tangentMode=RCTM_User, time=1, value=0}
       ]
      }}})
   &GravityForce(Gravity=<0 0 -300>)
   &SolveForcesAndVelocity(Clamp Velocity=false, Limit Acceleration=false)
   &ScaleSpriteSizeBySpeed(Max Scale Factor=<1.6 5>, Min Scale Factor=<0.8 3>)
   &ScaleColor(
    Scale RGB=&MakeVectorFromLinearColorRGB(Color=&FNTimeOfDayColor(Night Color=<rgba 0.162 0.3316 0.5017 1>, Morning Color=<rgba 1 0.7745 0.7008 1>, Day Color=<rgba 1 1 1 1>, Evening Color=<rgba 1 1 1 1>))
    ScaleA=true
    ScaleRGB=true
   )
  ]
  rendererClasses [&NiagaraSpriteRendererProperties]
  renderers [{rendererClass=&NiagaraSpriteRendererProperties}]
 }
 Liquid {
  emitterUpdateScript [
   &EmitterState(
    Inactive Response=::NewEnumerator1
    Life Cycle Mode=::NewEnumerator1
    Loop Behavior=::NewEnumerator1
    Loop Delay=0.35
    Loop Duration=3
    UseLoopDelay=true
   )
   &SpawnBurst_Instantaneous(Use Spawn Probability=false, UseLoopCountLimit=false)
   /SetVariables(Emitter.Orientation:Quat4f=$Engine.Owner.Rotation) {isSetParameters=true}
   /SetVariables(Emitter.Position:Vector3f=/Niagara/DynamicInputs/Helpers/SimulationPosition()) {isSetParameters=true}
   /SetVariables(
    Emitter.RotationVelocity:Vector3f=&Divide_Vector(
     A:Vector3f=&Subtract_Vector(
      A=&RotateVectorByQuaternion(Vector To Rotate=<0 0 1>, Quaternion=$Emitter.Orientation)
      B=&RotateVectorByQuaternion(Vector To Rotate=<0 0 1>, Quaternion=$Emitter.Previous.Orientation)
     )
     B:Vector3f=&VectorFromFloat(Value=$Engine.DeltaTime)
    )
    Emitter.Velocity:Vector3f=&Divide_Vector(
     A:Vector3f=&Subtract_Vector(A=$Emitter.Position, B=$Emitter.Previous.Position)
     B:Vector3f=&VectorFromFloat(Value=$Engine.DeltaTime)
    )
    Emitter.Viscosity=0.6
   ) {isSetParameters=true}
   /SetVariables(Emitter.InheritSourceMovement:NiagaraFloat=&Lerp_Float(A=0, B=1, Alpha=$Emitter.Viscosity)) {isSetParameters=true}
  ]
  particleSpawnScript [&InitializeParticle(
    Color=&FNTimeOfDayColor(Night Color=<rgba 0.281 0.3676 0.5955 1>, Morning Color=<rgba 1 0.73 0.6627 1>, Day Color=<rgba 1 1 1 1>, Evening Color=<rgba 1 1 1 1>)
    Lifetime=3.3
    Mesh Scale Mode=::NewEnumerator3
    Mesh Uniform Scale=$User.LiquidScale
    UsePositionOffset=false
    Write Material Random=true
   )]
  particleUpdateScript [
   &ParticleState()
   &SkeletalMeshLocation(
    MaskSpawnGroup=false
    Mesh Sampling Type=::NewEnumerator5
    Orientation Sampling=::NewEnumerator1
    Position Sampling=::NewEnumerator2
    Sampled Position Offset=<0 0 -9>
    Skeletal Mesh=$System.Skeletal Mesh Reader
    Socket Sampling Mode=::NewEnumerator5
   )
   &SocketLocation(
    SamplePosOffset=true
    Sampled Position Offset=<0 0 -9.3>
    Socket Index=1
    Socket Reader=$System.Socket Reader
    Socket Sampling Mode=::NewEnumerator3
    VelocitySpeedLimit=false
   ) {enabled=false}
   /SetVariables(Particles.Position_Origin:NiagaraPosition=$Particles.SkeletalMeshLocation.SampledPosition) {isSetParameters=true}
   &PendulumSetup(
    Module.Enable Spring Driver=true
    Pendulum Length:NiagaraFloat=&Multiply_Float(A=16, B=$User.LiquidScale)
    Pendulum Pivot:NiagaraPosition=$Particles.Position_Origin
    Pendulum Rest Axis:Vector3f=&NormalizeVector(
     Vector To Normalize=&Add_Vector(
      A=<0 0 -1>
      B=&Lerp_Vector(
       A=<0 0 0>
       B=&TransformVector(Vector=<0 0 -1>, Destination Space=::Simulation)
       Alpha=$System.Timeline.Float
      )
     )
     Safe Normalize=true
     Safe Fallback Threshold=0
     Safe Fallback Vector=<0 0 0.001>
    )
    "Spring Driven Constraint (Tightness)"=1.5
   )
   &GravityForce(Gravity=&Lerp_Vector(A=<0 0 -980>, B=<0 0 -19.6>, Alpha=$Emitter.Viscosity))
   &Drag(Drag=&Lerp_Float(A=0, B=4, Alpha=$Emitter.Viscosity), Use Linear Drag=true, Use Rotational Drag=false)
   &InheritSourceMovement(
    Applied Position Scale=&VectorFromFloat(Value=$Emitter.InheritSourceMovement)
    ApplyForce=false
    ApplyPosition=true
    ApplyPositionFalloff=false
    Source Velocity:Vector3f=$Emitter.Velocity
   )
   &SolveForcesAndVelocity(Clamp Velocity=false, Limit Acceleration=false, Speed Limit=200)
   &PendulumConstraint()
   &CalculateAccurateVelocity(Previous Position:NiagaraPosition=&Add_Vector(
     A=Linked() {linkedVariable {name=Particles.Previous.Position, type=NiagaraPosition}}
     B=&Multiply_VectorByFloat(Vector=$Emitter.Velocity, Float=&Multiply_Float(A=$Emitter.InheritSourceMovement, B=$Engine.DeltaTime))
    ))
   /Niagara/Modules/Update/Orientation/V2/OrientMeshToVector(
    Look At Direction:Vector3f=&NormalizeVector(
     Vector To Normalize=&Subtract_Vector(
      A=Linked() {linkedVariable {name=Particles.Position_Origin, type=NiagaraPosition}}
      B=Linked() {linkedVariable {name=Particles.Position, type=NiagaraPosition}}
     )
     Safe Normalize=true
     Safe Fallback Threshold=0
     Safe Fallback Vector=<0 0 0.001>
    )
    Override Quat=false
    Reference Axis Vector:Vector3f=<0 0 1>
    Up Axis Direction:Vector3f=<0 0 1>
    UseOptionalSideAxis=false
    UseOptionalUpAxis=false
   )
   &InterpolateOverTime(
    Rate Of Change=2
    Target Float Value:NiagaraFloat=&Max_Float(
     A:NiagaraFloat=&RemapRange(
      Input Value To Remap=&DotProduct(
       Vector A=<0 0 -1>
       Vector B=&ReturnMeshOrientationAxis(Mesh Orientation Axis=::ZAxis, Mesh Orientation Quat=$Particles.MeshOrientation)
      )
      Input Min=-1
      Input Max=0
      Output Min=0
      Output Max=1
     )
     B:NiagaraFloat=&RemapRange(
      Input Value To Remap=&Divide_Float(A=&Subtract_Float(A=$Particles.DistanceTraveled, B=$Particles.Previous.DistanceTraveled), B=$User.LiquidScale)
      Input Min=2
      Input Max=4
      Output Min=0
      Output Max=1
     )
    )
   )
   /SetVariables(Particles.LiquidSpeed:NiagaraFloat=&ClampFloat(Float=$Particles.InterpolateOverTime.InterpolateOverTime_Float.Moving Average)) {isSetParameters=true}
   &Timeline(Play=&SetBoolByFloatComparison(A=$Particles.LiquidSpeed, B=0.333), Play Rate=3, Timeline End=10)
   /SetVariables(Particles.Turbulence:NiagaraFloat=$Particles.Timeline.NormalizedTime) {isSetParameters=true}
   /Niagara/Modules/Update/Utility/IncrementOverTime(Rate of Change:NiagaraFloat=&Lerp_Float(A=1, B=1.25, Alpha=$Particles.LiquidSpeed))
   /SetVariables(
    Particles.DynamicMaterialParameter:Vector4f=&MakeFloat4FromVectorAndFloat(
     "Vector (RGB)"=&ReturnMeshOrientationAxis(Mesh Orientation Axis=::ZAxis, Mesh Orientation Quat=$Particles.MeshOrientation)
     "Float (Alpha)"=&FloatFromCurve(
      FloatCurve=DataInterface() {propertyValues {Curve {
         defaultValue=1
         keys [
          {arriveTangent=0.7649, interpMode=RCIM_Cubic, leaveTangent=0.7649, tangentMode=RCTM_User, time=0, value=0.05}
          {
           arriveTangent=3.9456
           arriveTangentWeight=0.6554
           interpMode=RCIM_Cubic
           leaveTangentWeight=0.1301
           tangentMode=RCTM_Break
           time=0.3624
           value=0.9918
          }
          {arriveTangent=0.1695, interpMode=RCIM_Cubic, leaveTangent=-7.1625, tangentMode=RCTM_Break, time=0.7, value=1.0039}
          {
           arriveTangentWeight=0.1383
           interpMode=RCIM_Cubic
           leaveTangent=-5.2682
           leaveTangentWeight=0.1181
           tangentMode=RCTM_Break
           time=0.9925
           value=-0.0356
          }
         ]
        }}}
      Scale Curve=0.85
     )
    )
    Particles.DynamicMaterialParameter1:Vector4f=&MakeFloat4FromVectorAndFloat(
     "Vector (RGB)"=&Multiply_VectorByFloat(
      Vector=&ReturnMeshOrientationAxis(Mesh Orientation Axis=::YAxis, Mesh Orientation Quat=$Particles.MeshOrientation)
      Float=-1
     )
     "Float (Alpha)"=$Particles.IncrementOverTime.Counter
    )
    Particles.DynamicMaterialParameter2:Vector4f=&MakeVector4(X=$Particles.LiquidSpeed, Y=$Particles.Turbulence, Z=0, W=0)
   ) {isSetParameters=true}
  ]
  rendererClasses [&NiagaraMeshRendererProperties]
  renderers [{rendererClass=&NiagaraMeshRendererProperties, rendererIndex=0}]
 }
 LiquidTrail {
  emitterUpdateScript [
   &EmitterState(Life Cycle Mode=::NewEnumerator1, Loop Behavior=::NewEnumerator1, Loop Duration=1.07, UseLoopDelay=false)
   &SpawnRate(SpawnRate=100, Use Spawn Probability=false)
  ]
  particleSpawnScript [
   &InitializeParticle(
    Color Mode=::NewEnumerator0
    Lifetime=0.6
    Ribbon Width=7
    Ribbon Width Mode=::NewEnumerator1
    Sprite Size Mode=::NewEnumerator3
    Uniform Sprite Size=5
    UsePositionOffset=false
    Write Material Random=true
   )
   &CameraOffset(Camera Offset Amount=1.5)
  ]
  particleUpdateScript [
   &ParticleState()
   &SocketLocation(
    Interpolate Socket Sampling=true
    Orientation Sampling=::NewEnumerator1
    Position Sampling=::NewEnumerator1
    SamplePosOffset=true
    Sampled Position Offset=<0 0 -2>
    Socket Reader=$System.Socket Reader
    Socket Sampling Mode=::NewEnumerator3
    VelocitySpeedLimit=false
   )
   &SocketLocation(
    Interpolate Socket Sampling=true
    Orientation Sampling=::NewEnumerator1
    Position Sampling=::NewEnumerator1
    SamplePosOffset=true
    Sampled Position Offset=<0 0 -1>
    Socket Index=2
    Socket Reader=$System.Socket Reader
    Socket Sampling Mode=::NewEnumerator3
    VelocitySpeedLimit=false
   ) {module=SocketLocation001}
   /Game/Effects/Niagara/Modules/Curves/FN_HermiteSpline(
    Direction End:Vector3f=<0 0 270>
    Direction Start:Vector3f=<0 0 250>
    Position Blend:NiagaraFloat=$Particles.NormalizedAge
    Position End:Vector3f=&MakeCustomVectorFromBool(
     Bool=$System.Debug
     True Vector=$System.EnPos
     False Vector=Linked() {linkedVariable {name=Output.SocketLocation001.SampledPosition, type=NiagaraPosition}}
    )
    Position Start:Vector3f=&MakeCustomVectorFromBool(
     Bool=$System.Debug
     True Vector=$System.StartPos
     False Vector=Linked() {linkedVariable {name=Output.SocketLocation.SampledPosition, type=NiagaraPosition}}
    )
    Write Particles Position=true
   )
   &ScaleRibbonWidth(Ribbon Width Scale=&FloatFromCurve(FloatCurve=DataInterface() {propertyValues {Curve {keys [
         {arriveTangent=1.6825, interpMode=RCIM_Cubic, leaveTangent=1.6825, tangentMode=RCTM_User, time=0.0066, value=0.8051}
         {arriveTangent=0.0974, interpMode=RCIM_Cubic, leaveTangent=0.0974, tangentMode=RCTM_User, time=0.4982, value=1.0265}
         {arriveTangent=0.5757, interpMode=RCIM_Cubic, leaveTangent=0.5757, time=0.8267, value=1.2027}
         {arriveTangent=0.5412, interpMode=RCIM_Cubic, leaveTangent=0.5412, tangentMode=RCTM_User, time=1.0065, value=1.3191}
        ]}}}))
   &ScaleColor(
    Scale Alpha=&FloatFromCurve(FloatCurve=DataInterface() {propertyValues {Curve {keys [
         {interpMode=RCIM_Cubic, leaveTangentWeight=0.0416, tangentMode=RCTM_Break, time=0.8503, value=1}
         {arriveTangentWeight=0.0693, interpMode=RCIM_Cubic, leaveTangentWeight=0.014, tangentMode=RCTM_Break, time=0.9751, value=0.9936}
         {arriveTangent=-1.6057, tangentMode=RCTM_User, time=1.0002, value=0.88}
        ]}}})
    Scale RGB=&MakeVectorFromLinearColorRGB(Color=&FNTimeOfDayColor(Night Color=<rgba 0.329 0.6669 1 1>, Morning Color=<rgba 1 0.7745 0.7008 1>, Day Color=<rgba 1 1 1 1>, Evening Color=<rgba 1 1 1 1>))
    ScaleA=true
    ScaleRGB=true
   )
  ]
  rendererClasses [&NiagaraSpriteRendererProperties, &NiagaraRibbonRendererProperties]
  renderers [
   {rendererClass=&NiagaraSpriteRendererProperties}
   {rendererClass=&NiagaraRibbonRendererProperties, rendererIndex=1}
   {rendererClass=&NiagaraRibbonRendererProperties, rendererIndex=2}
  ]
 }
 Minimal {
  emitterUpdateScript [
   &EmitterState()
   &SpawnParticlesFromOtherEmitter(
    Attribute Reader=DataInterface() {propertyValues {EmitterBinding {bindingMode=Other, emitterName=TrailDropsLead}}}
    Spawn Rate=35
    Use Maximum Age=false
    Use Minimum Age=false
   )
  ]
  particleSpawnScript [
   &InitializeParticle(
    Color Mode=::NewEnumerator0
    Lifetime Max=0.6
    Lifetime Min=0.2
    Lifetime Mode=::NewEnumerator1
    Sprite Size Mode=::NewEnumerator1
    Uniform Sprite Size Max=1.8
    Uniform Sprite Size Min=0.8
    UsePositionOffset=false
    Write Material Random=true
   )
   &SampleParticlesFromOtherEmitter(Applied Velocity Scale=0.3, Velocity Sampling=::NewEnumerator0)
   &ShapeLocation(MaskSpawnGroup=false, Override Seed=false, Sphere Radius=3, UseOwnerScale=true)
   &AddVelocity(Velocity=$Output.SampleParticlesFromOtherEmitter.SampledVelocity)
  ]
  particleUpdateScript [
   &ParticleState()
   &ScaleSpriteSize(Uniform Curve Sprite Scale=DataInterface() {propertyValues {Curve {
       defaultValue=0
       keys [
        {arriveTangent=9.0725, interpMode=RCIM_Cubic, leaveTangent=9.0725, tangentMode=RCTM_User, time=0, value=0}
        {arriveTangent=0.1451, interpMode=RCIM_Cubic, leaveTangent=0.1451, tangentMode=RCTM_User, time=0.1408, value=0.9847}
        {arriveTangent=-2.3613, interpMode=RCIM_Cubic, leaveTangent=-2.3613, tangentMode=RCTM_User, time=1.0054, value=0.5256}
       ]
      }}})
   &GravityForce(Gravity=<0 0 -650>)
   &Drag(Drag=5, Use Linear Drag=true, Use Rotational Drag=true)
   &SolveForcesAndVelocity(Clamp Velocity=false, Limit Acceleration=false)
   &ScaleSpriteSizeBySpeed(Max Scale Factor=<1.6 3>, Min Scale Factor=<1 2>)
   &ScaleColor(
    Scale RGB=&MakeVectorFromLinearColorRGB(Color=&FNTimeOfDayColor(Night Color=<rgba 0.162 0.3316 0.5017 1>, Morning Color=<rgba 1 0.7745 0.7008 1>, Day Color=<rgba 1 1 1 1>, Evening Color=<rgba 1 1 1 1>))
    ScaleA=true
    ScaleRGB=true
   )
  ]
  rendererClasses [&NiagaraSpriteRendererProperties]
  renderers [
   {rendererClass=&NiagaraSpriteRendererProperties}
   {rendererClass=&NiagaraSpriteRendererProperties, rendererIndex=1}
  ]
 }
 Splashes {
  emitterUpdateScript [
   &EmitterState(Life Cycle Mode=::NewEnumerator1, Loop Behavior=::NewEnumerator1, Loop Delay=0.55, Loop Duration=0.9, UseLoopDelay=true)
   &SpawnBurst_Instantaneous(Spawn Count=2, Use Spawn Probability=false, UseLoopCountLimit=false)
   &SpawnRate(SpawnRate=8, Use Spawn Probability=false)
  ]
  particleSpawnScript [
   &InitializeParticle(
    Color Mode=::NewEnumerator0
    Lifetime Max=0.4
    Lifetime Min=0.25
    Lifetime Mode=::NewEnumerator1
    Sprite Rotation Mode=::NewEnumerator2
    Sprite Size Mode=::NewEnumerator1
    Uniform Sprite Size Max=13
    Uniform Sprite Size Min=9
    UsePositionOffset=false
    Write Material Random=true
   )
   &SocketLocation(SamplePosOffset=false, Socket Index=1, Socket Reader=$System.Socket Reader, Socket Sampling Mode=::NewEnumerator3, VelocitySpeedLimit=false)
   &ShapeLocation(MaskSpawnGroup=false, Override Seed=false, Sphere Radius=3, UseOwnerScale=true) {enabled=false}
   &AddVelocity(
    Module.Use Falloff Distance=false
    Origin Offset=<0 0 -1>
    Override Seed=false
    Velocity Mode=::NewEnumerator1
    Velocity Origin=$Output.SocketLocation.SampledPosition
    Velocity Speed=&RandomRangeFloat(Minimum=5, Maximum=12)
   )
   &SubUVAnimation(
    "Sprite Renderer "=DataInterface() {propertyValues {SpriteRenderer=&self:Splashes_0.NiagaraSpriteRendererProperties_0}}
    SubUV Animation Mode:ENiagara_SubUVLookupModeV2=::NewEnumerator2
    UseEndFrame=false
    UseStartFrame=false
   )
  ]
  particleUpdateScript [
   &ParticleState()
   &ScaleSpriteSize(Uniform Curve Sprite Scale=DataInterface() {propertyValues {Curve {
       defaultValue=0
       keys [
        {arriveTangent=9.0725, interpMode=RCIM_Cubic, leaveTangent=9.0725, tangentMode=RCTM_User, time=0, value=0}
        {arriveTangent=0.2183, interpMode=RCIM_Cubic, leaveTangent=0.2183, tangentMode=RCTM_User, time=0.2999, value=1.0751}
       ]
      }}})
   &GravityForce(Gravity=<0 0 -120>) {enabled=false}
   &SolveForcesAndVelocity(Clamp Velocity=false, Limit Acceleration=false)
   &ScaleSpriteSizeBySpeed(Max Scale Factor=<2 2.2>, Min Scale Factor=<1 1.2>)
   &DynamicMaterialParameters(
    Index 0 Param 4=&FloatFromCurve(FloatCurve=DataInterface() {propertyValues {Curve {
        defaultValue=0.9847
        keys [{interpMode=RCIM_Cubic, tangentMode=RCTM_User, time=0, value=0}, {interpMode=RCIM_Cubic, tangentMode=RCTM_User, time=1, value=1}]
       }}})
    Module.Index0Float0Write=true
    Module.Index0Float1Write=true
    Module.Index0Float2Write=true
    Module.Index0Float3Write=true
   )
   &ScaleColor(
    Scale RGB=&MakeVectorFromLinearColorRGB(Color=&FNTimeOfDayColor(Night Color=<rgba 0.0956 0.2262 0.3698 1>, Morning Color=<rgba 1 0.7745 0.7008 1>, Day Color=<rgba 1 1 1 1>, Evening Color=<rgba 1 1 1 1>))
    ScaleA=true
    ScaleRGB=true
   )
  ]
  rendererClasses [&NiagaraSpriteRendererProperties]
  renderers [{rendererClass=&NiagaraSpriteRendererProperties}]
 }
 TrailDropsLead {
  emitterUpdateScript [
   &EmitterState(Life Cycle Mode=::NewEnumerator1, Loop Behavior=::NewEnumerator1, Loop Delay=0.12, Loop Duration=1.5, UseLoopDelay=true)
   &SpawnBurst_Instantaneous(Use Spawn Probability=false, UseLoopCountLimit=false)
  ]
  particleSpawnScript [
   &InitializeParticle(
    Color Mode=::NewEnumerator0
    Lifetime=1.5
    Ribbon Width=8
    Ribbon Width Mode=::NewEnumerator1
    Sprite Size Mode=::NewEnumerator3
    Uniform Sprite Size=15
    UsePositionOffset=false
    Write Material Random=true
   )
   &ShapeLocation(MaskSpawnGroup=false, Override Seed=false, Shape Origin:Vector3f=$Output.FN_HermiteSpline.Position, Sphere Radius=5, UseOwnerScale=true)
  ]
  particleUpdateScript [
   &ParticleState()
   &SocketLocation(
    Interpolate Socket Sampling=true
    Orientation Sampling=::NewEnumerator1
    Position Sampling=::NewEnumerator1
    SamplePosOffset=true
    Socket Reader=$System.Socket Reader
    Socket Sampling Mode=::NewEnumerator3
    VelocitySpeedLimit=false
   )
   &SocketLocation(
    Interpolate Socket Sampling=true
    Orientation Sampling=::NewEnumerator1
    Position Sampling=::NewEnumerator1
    SamplePosOffset=true
    Sampled Position Offset=<0 0 -6>
    Socket Index=1
    Socket Reader=$System.Socket Reader
    Socket Sampling Mode=::NewEnumerator3
    VelocitySpeedLimit=false
   ) {module=SocketLocation001}
   &FN_HermiteSpline(
    Compute Velocity=true
    Direction End:Vector3f=<0 0 250>
    Direction Start:Vector3f=<0 0 250>
    Position Blend:NiagaraFloat=&FloatFromCurve(
     FloatCurve=DataInterface() {propertyValues {Curve {keys [
         {interpMode=RCIM_Cubic, leaveTangent=0.5263, tangentMode=RCTM_Break, time=0, value=0}
         {arriveTangent=1.5709, tangentMode=RCTM_User, time=1, value=1}
        ]}}}
     CurveIndex=$Emitter.NormalizedLoopAge
    )
    Position End:Vector3f=&MakeCustomVectorFromBool(
     Bool=$System.Debug
     True Vector=$System.EnPos
     False Vector=Linked() {linkedVariable {name=Output.SocketLocation001.SampledPosition, type=NiagaraPosition}}
    )
    Position Start:Vector3f=&MakeCustomVectorFromBool(
     Bool=$System.Debug
     True Vector=$System.StartPos
     False Vector=Linked() {linkedVariable {name=Output.SocketLocation.SampledPosition, type=NiagaraPosition}}
    )
    Write Particles Position=true
    Write Particles Velocity=true
   )
   &ScaleRibbonWidth(Ribbon Width Scale=&FloatFromCurve(FloatCurve=DataInterface() {propertyValues {Curve {keys [
         {interpMode=RCIM_Cubic, tangentMode=RCTM_User, time=0.0022, value=0.328}
         {interpMode=RCIM_Cubic, tangentMode=RCTM_User, time=0.1923, value=0.6382}
         {interpMode=RCIM_Cubic, tangentMode=RCTM_User, time=0.5069, value=0.9565}
         {interpMode=RCIM_Cubic, tangentMode=RCTM_User, time=1.0011, value=1.2196}
        ]}}}))
   /SetVariables(Particles.Velocity=&Multiply_Vector(A=$Particles.Initial.Velocity, B=<0.2 0.2 0.2>)) {enabled=false, isSetParameters=true}
   &ScaleColor(
    Scale RGB=&MakeVectorFromLinearColorRGB(Color=&FNTimeOfDayColor(Night Color=<rgba 0.162 0.3316 0.5017 1>, Morning Color=<rgba 1 0.7745 0.7008 1>, Day Color=<rgba 1 1 1 1>, Evening Color=<rgba 1 1 1 1>))
    ScaleA=true
    ScaleRGB=true
   )
  ]
  rendererClasses [&NiagaraSpriteRendererProperties]
  renderers [{rendererClass=&NiagaraSpriteRendererProperties}]
 }
}
summary {
 systemName=NS_Emote_QuicheLorraine_Slurp
 emitters [
  {emitterName=LiquidTrail, rendererClasses [&NiagaraSpriteRendererProperties, &NiagaraRibbonRendererProperties]}
  {emitterName=GlassDrops, rendererClasses [&NiagaraSpriteRendererProperties]}
  {emitterName=Liquid, rendererClasses [&NiagaraMeshRendererProperties]}
  {emitterName=Splashes, rendererClasses [&NiagaraSpriteRendererProperties]}
  {emitterName=TrailDropsLead, rendererClasses [&NiagaraSpriteRendererProperties]}
  {emitterName=GlassDrops001, rendererClasses [&NiagaraSpriteRendererProperties]}
  {emitterName=GlassDrops002, rendererClasses [&NiagaraSpriteRendererProperties]}
  {emitterName=Minimal, rendererClasses [&NiagaraSpriteRendererProperties]}
 ]
 userVariables [
  User.LiquidScale {type=NiagaraFloat, defaultValue=1, description=""}
  User.NormalizedNotifyProgress {type=NiagaraFloat, defaultValue=0, description=""}
 ]
}
